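; Creative Class Settings device that defines the options and item loadout for one minigame class slot.
; It registers its options with the minigame's class and weapon settings mutators and supplies a class-scoped data registry resolver.
; Pickups hitting the drop area before game start fill the item list on the server.
; Spawn and class delegates drive granting, auto-equipping and removal of items, with player ID sets enforcing the respawn regrant option.
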(event EventBeginPlay
  ;[0 1 0.04321 0.5] Initialization
  (|CallParentFunction)
  (Networking|SwitchHasAuthority
    (:Remote
      (Initialization|InitializeComponent (Variables|Default|GetCreative_ItemPreview_Component) (Variables|Default|GetPreviewMeshScene) (Variables|Default|GetPreviewStaticMesh) (Variables|Default|GetPreviewSkeletalMesh) (Variables|Default|GetAmbient%20Loop) (Variables|Default|GetPickupItemList) (Variables|Default|GetItemVariantNotifier) :!class /Game/Creative/Devices/Common/Components/Creative_ItemPreview_Component.Creative_ItemPreview_Component_C))))

(event MinigameEndedDelegate(FortMinigameProgress)
  ;[0.5156 0 0.008952 0.2] Deconstruction - Clean Up
  (Utilities|Set|Clear (Variables|Default|GetGrantedPlayers))
  (Utilities|Set|Clear (Variables|Default|GetInitialized%20Players)))

(event Custom|Remove%20Player
  (params (UniqueNetId :type Unique%20Net%20Id%20Structure) (bIsLocalPlayer :type Boolean))
  ;[0.5156 0 0.008952 0.2] Deconstruction - Clean Up
  ;;[0.03646 0 0.000633 0] If a player is removed from the minigame, also remove them from the list of granted players.
  (Utilities|Set|Remove (Variables|Default|GetGrantedPlayers) UniqueNetId)
  (Utilities|Set|Remove (Variables|Default|GetInitialized%20Players) UniqueNetId))

(event OnRemovedfromMinigame(FortMinigameProgress)
  (params (Minigame :type Fort%20Minigame%20Object%20Reference))
  ;[0.5156 0 0.008952 0.2] Deconstruction - Clean Up
  (ClassSettings|RemoveClassSettingsfromMinigame Minigame)
  (CallFunction|BindorUnbindDelegates false Minigame))

(event OnAddedtoMinigame(FortMinigameProgress)
  (params (Minigame :type Fort%20Minigame%20Object%20Reference))
  ;[0 1 0.04321 0.5] Initialization
  (ClassSettings|HandleAddorRemoveClassSettings Minigame)
  (CallFunction|BindorUnbindDelegates true Minigame)
  (CallFunction|OnPlayModeChanged Minigame (Minigame|IsinPlayMode Minigame :!class /Script/FortniteGame.FortMinigame)))

(event EventOnPlayModeChanged
  (params (Minigame :type Fort%20Minigame%20Object%20Reference) (bIsInPlayMode :type Boolean))
  ;[1 0.1082 0 0.25] Disable Collision on drop area during Edit mode, as overlaps are expensive.
  (|CallParentFunction Minigame bIsInPlayMode)
  (Collision|SetCollisionEnabled (Variables|Default|GetDropArea) (select bIsInPlayMode "NoCollision" "QueryOnly")))

(event Custom|Player%20Spawned
  (params (SpawningPlayer :type Controller%20Object%20Reference))
  ;[0 0.8171 0.8594 0.3] Player Spawn Logic - Flag as Initialized and Attempt to Spawn Items
  (bind _playerstate (Variables|/Script/Engine.Controller|GetPlayerState SpawningPlayer)
    (:then
      (bind _asfort_player_state (Utilities|Casting|CastToFortPlayerState _playerstate))
      (if (CallFunction|IsDeviceValidForPlayerState _asfort_player_state)
        (Utilities|Set|Add (Variables|Default|GetInitialized%20Players) (PlayerState|GetUniqueID _asfort_player_state))
        (CallFunction|AttemptGrantItems (Pawn|GetControlledPawn SpawningPlayer))))
    (:else)))

(event OnAnyPropertyChanged(ToyOptionsComponent)
  ;[0 1 0.04321 0.5] Initialization
  (ClassSettings|HandleAddorRemoveClassSettings (Minigame|GetCurrentMinigame)))

(event OnPropertiesLoaded(ToyOptionsComponent)
  ;[0 1 0.04321 0.5] Initialization
  (ClassSettings|HandleAddorRemoveClassSettings (Minigame|GetCurrentMinigame)))

(event Custom|OnPlayerClassApplied
  (params (PlayerState :type Fort%20Player%20State%20Object%20Reference) (ClassSlotIndex :type Byte))
  (bind _returnvalue (PlayerState|GetCurrentPawn PlayerState))
  ;[0 0.4352 0.8594 0.2] Class Application / Removal
  (Utilities|IsValid PlayerState
    (:Is%20Valid
      (Utilities|IsValid _returnvalue
        (:Is%20Valid
          (if (Math|Boolean|ANDBoolean (not (Variables|/Script/FortniteGame.FortPawn|GetbIsDying _returnvalue)) (CallFunction|IsThisClass ClassSlotIndex) (Utilities|Set|ContainsItem (Variables|Default|GetInitialized%20Players) (PlayerState|GetUniqueID PlayerState)))
            (CallFunction|AttemptGrantItems _returnvalue)))))))

(event OnComponentHit(DropArea)
  (params (HitComponent :type Primitive%20Component%20Object%20Reference) (OtherActor :type Actor%20Object%20Reference) (OtherComp :type Primitive%20Component%20Object%20Reference) (NormalImpulse :type Vector :default "0, 0, 0") (Hit :type Hit%20Result%20Structure%20%28by%20ref%29))
  ;[0 0.07202 1 0.5] OnComponentHit(DropArea)
  (Networking|SwitchHasAuthority
    (:Authority
      (Utilities|IsValid OtherActor
        (:Is%20Valid
          (CallFunction|HandleDropAreaOverlap OtherActor))))))

(event OnItemAdded(PickupItemList)
  (params (ItemAdded :type Actor%20Object%20Reference))
  (CallFunction|Resave))

(event Custom|OnPlayerClassRemoved
  (params (PlayerState :type Fort%20Player%20State%20Object%20Reference) (ClassSlotIndex :type Byte))
  (bind _returnvalue (Fortnite|GetFortPlayerControllerfromActor PlayerState))
  ;[0 0.4352 0.8594 0.2] Class Application / Removal
  (Utilities|IsValid PlayerState
    (:Is%20Valid
      (if (CallFunction|IsThisClass ClassSlotIndex)
        (Utilities|IsValid _returnvalue
          (:Is%20Valid
            (Inventory|RemoveItemsfromPlayer _returnvalue)
            (Utilities|Set|Remove (Variables|Default|GetGrantedPlayers) (PlayerState|GetUniqueID PlayerState))))))))

(event OnItemListUpdated(PickupItemList)
  (bind _minigame (Minigame|GetCurrentMinigame))
  (Utilities|IsValid _minigame
    (:Is%20Valid
      (Networking|SwitchHasAuthority
        (:Authority
          (Minigame|GameRules|RefreshGameRulesPreview _minigame))))))

(event Custom|ShowInUIWhenReceivedFrom_Func
  (params (Instigator :type Controller%20Object%20Reference))
  (bind _minigame (Minigame|GetCurrentMinigame))
  (bind _returnvalue (Minigame|GetFNEClassComponent _minigame))
  (Utilities|IsValid _minigame
    (:Is%20Valid
      (Utilities|IsValid _returnvalue
        (:Is%20Valid
          (Minigame|GameRules|SetClassSlotVisibleinUI _returnvalue (Math|Conversions|ToByte(Integer) (Variables|DeviceOptions|GetClassIndex)) true))))))

(event Custom|HideInUIWhenReceivedFrom_Func
  (params (Instigator :type Controller%20Object%20Reference))
  (bind _minigame (Minigame|GetCurrentMinigame))
  (bind _returnvalue (Minigame|GetFNEClassComponent _minigame))
  (Utilities|IsValid _minigame
    (:Is%20Valid
      (Utilities|IsValid _returnvalue
        (:Is%20Valid
          (Minigame|GameRules|SetClassSlotVisibleinUI _returnvalue (Math|Conversions|ToByte(Integer) (Variables|DeviceOptions|GetClassIndex)) false))))))

(event Custom|RPC_MULTICAST%20Add%20Item
  (params (Item :type Fort%20Pickup%20Athena%20Object%20Reference))
  (Utilities|IsValid Item
    (:Is%20Valid
      (Networking|CanExecuteCosmeticEvents
        (:True
          (Class|CreativeItemPreviewComponent|ItemAddedToQueueFX (Variables|Default|GetCreative_ItemPreview_Component) Item))))))

(fn AddClassSettingsToMinigame
  (params (Minigame :type Fort%20Minigame%20Object%20Reference) (ClassSlotIndex :type Byte))
  (bind _classslotindex (Variables|GetClassSlotIndex))
  (Utilities|FlowControl|Sequence
    (:then_0
      (bind _output_get (LocalVariables|SetSettingsMutator (ClassSettings|GetClassSettingsMutator (Variables|GetMinigame))))
      (Utilities|IsValid _output_get
        (:Is%20Valid
          (bind _returnvalue (ClassSettings|AddDataSourceforPlayerClass (LocalVariables|GetSettingsMutator) _classslotindex (Variables|Default|GetToyOptionsComponent)))
          (bind _returnvalue_2 (ClassSettings|AddDataSourceforPlayerClass (LocalVariables|GetSettingsMutator) _classslotindex self))
          (if (or _returnvalue _returnvalue_2)
            (Minigame|GameRules|RefreshGameRulesPreview (Variables|GetMinigame))))))
    (:then_1
      (bind _output_get_2 (LocalVariables|SetWeaponSettingsMutator (ClassSettings|GetWeaponSettingsMutator (Variables|GetMinigame))))
      (Utilities|IsValid _output_get_2
        (:Is%20Valid
          (Class|FortAthenaMutatorWeaponSettings|AddDataSourceforClassSlot _output_get_2 "/Game/Athena/Items/Cosmetics/Pickaxes/DefaultPickaxe.DefaultPickaxe" (Variables|GetClassSlotIndex) (Variables|Default|GetToyOptionsComponent)))))
    (:then_2
      (Class|FortAthenaMutatorUtilities|RemoveDataRegistryOptionsforContextObject)
      (Class|FortAthenaMutatorUtilities|ApplyActorOptionsComponenttoDataRegistry (Variables|Default|GetToyOptionsComponent)))))

(fn AddItemToQueue
  (params (NewItem :type Fort%20Pickup%20Athena%20Object%20Reference))
  (Utilities|IsValid NewItem
    (:Is%20Valid
      (bind _bitemadded (Class|PickupItemListComponent|AddItemtoQueue (Variables|Default|GetPickupItemList) NewItem))
      (if _bitemadded
        (CallFunction|RPCMULTICASTAddItem NewItem)))))

(fn AttemptGrantItems
  (params (Possessed%20Pawn :type Pawn%20Object%20Reference) (ForceGrantItems :type Boolean))
  (bind _asfort_player_pawn (Utilities|Casting|CastToFortPlayerPawn Possessed%20Pawn))
  (bind _asfort_player_state (Utilities|Casting|CastToFortPlayerState (Variables|/Script/Engine.Controller|GetPlayerState (Pawn|GetController _asfort_player_pawn))))
  (if (CallFunction|IsDeviceValidForPlayerState _asfort_player_state)
    (if (ClassSettings|IsPlayerMatchingClass _asfort_player_state)
      (if (or (Inventory|ShouldGrantItems (PlayerState|GetUniqueID _asfort_player_state)) (Variables|GetForceGrantItems))
        (bind _asfort_player_controller_athena (Utilities|Casting|CastToFortPlayerControllerAthena (Pawn|GetController _asfort_player_pawn)))
        (Inventory|TryGiveAllItemstoPlayer _asfort_player_controller_athena)
        (Utilities|Set|Add (Variables|Default|GetGrantedPlayers) (PlayerState|GetUniqueID _asfort_player_state)))
      (else
        _))
    (else
      _))
  (Utilities|Set|Remove (Variables|Default|GetGrantedPlayers) (PlayerState|GetUniqueID _asfort_player_state)))

(fn BindOrUnbindDelegates
  (params (Bind :type Boolean) (Minigame :type Fort%20Minigame%20Object%20Reference))
  (bind _outputdelegate (EventDispatchers|CreateEvent :!event Player%20Spawned))
  (bind _outputdelegate_2 (EventDispatchers|CreateEvent :!event OnPlayerClassRemoved))
  (bind _outputdelegate_3 (EventDispatchers|CreateEvent :!event OnPlayerClassApplied))
  (bind _outputdelegate_4 (EventDispatchers|CreateEvent :!event Remove%20Player))
  (bind _minigame (Variables|GetMinigame))
  (bind _minigame_2 (Variables|GetMinigame))
  (Utilities|IsValid Minigame
    (:Is%20Valid
      (if Bind
        (Minigame|BindEventtoOnMinigamePlayerSpawned _minigame_2 _outputdelegate)
        (Minigame|BindEventtoOnPlayerClassRemoved _minigame_2 _outputdelegate_2)
        (Minigame|BindEventtoOnPlayerClassApplied _minigame_2 _outputdelegate_3)
        (Minigame|BindEventtoPlayerRemovedDelegate _minigame_2 _outputdelegate_4)
        (else
          (Minigame|UnbindEventfromOnMinigamePlayerSpawned _minigame _outputdelegate)
          (Minigame|UnbindEventfromOnPlayerClassRemoved _minigame _outputdelegate_2)
          (Minigame|UnbindEventfromOnPlayerClassApplied _minigame _outputdelegate_3)
          (Minigame|UnbindEventfromPlayerRemovedDelegate _minigame _outputdelegate_4))))))

(fn BuildDataRegistryResolverScope
  (Utilities|FlowControl|Sequence
    (:then_0
      (bind _output_get (LocalVariables|SetSpatialInterfaceName (SpatialGameplay|GetSpatialGameplayInterfaceFNameforActor self)))
      (if (!= _output_get (LocalVariables|GetNameNone))
        (Utilities|Array|Add (LocalVariables|GetOutNameScopes) _output_get)))
    (:then_1
      (Utilities|Array|Add (LocalVariables|GetOutNameScopes) (Utilities|String|StringToName (Utilities|String|BuildString(Integer) "Class" "" (Variables|DeviceOptions|GetClassIndex))))
      (LocalVariables|SetOutPriority (+ (LocalVariables|GetOutPriority) 1)))
    (:then_2
      (return (LocalVariables|GetOutNameScopes) (LocalVariables|GetOutPriority) (> (Utilities|Array|Length (LocalVariables|GetOutNameScopes)) 0)))))

(fn GetCacheContextOverride
  (return 0))

(fn GetClassSettingsMutator
  (params (Minigame :type Fort%20Minigame%20Object%20Reference))
  (Utilities|FlowControl|Sequence
    (:then_0
      (bind _output_get (LocalVariables|SetCurrentMinigame Minigame))
      (Utilities|IsValid _output_get
        (:Is%20Valid
          (bind _output_get_2 (LocalVariables|SetMutatorList (Actor|GetComponentbyClass _output_get "/Script/FortniteGame.FortMutatorListComponent")))
          (Utilities|IsValid _output_get_2
            (:Is%20Valid
              (bind _asfort_athena_mutator_player_class_settings (Utilities|Casting|CastToFortAthenaMutator_PlayerClassSettings (Modifier|GetMutatorbyClass _output_get_2 "/Script/FortniteGame.FortAthenaMutator_PlayerClassSettings")))
              (LocalVariables|SetSettingsMutator _asfort_athena_mutator_player_class_settings))))))
    (:then_1
      (return (LocalVariables|GetSettingsMutator)))))

(fn GetCreativeClassInventory
  (bind _pickupitemlist (Variables|Default|GetPickupItemList))
  (Utilities|IsValid _pickupitemlist
    (:Is%20Valid
      (return (ItemData|GetContainedItems _pickupitemlist)))
    (:Is%20Not%20Valid
      (return 0))))

(fn GetCreativeClassName
  (bind _minigame (Minigame|GetCurrentMinigame))
  (bind _returnvalue (Minigame|GetFNEClassComponent _minigame))
  (Utilities|IsValid _minigame
    (:Is%20Valid
      (Utilities|IsValid _returnvalue
        (:Is%20Valid
          (return (select (Utilities|Text|TextIsEmpty (Minigame|GameRules|GetClassSlotName _returnvalue (Math|Conversions|ToByte(Integer) (Variables|DeviceOptions|GetClassIndex)))) "NSLOCTEXT(\"[3F00A56742ECBCD0347F81BFC7532D10]\", \"C3F8BE164851DF8EABFCDFA431F1FDE6\", \"No Class Slot Set\")" (Minigame|GameRules|GetClassSlotName _returnvalue (Math|Conversions|ToByte(Integer) (Variables|DeviceOptions|GetClassIndex))))))
        (:Is%20Not%20Valid
          _)))
    (:Is%20Not%20Valid
      _))
  (return "NSLOCTEXT(\"[3F00A56742ECBCD0347F81BFC7532D10]\", \"F4D5D4434B8F135EA0279CB51C210C06\", \"No Class Info Available\")"))

(fn GetCurrentMinigame
  (return (Variables|/Script/FortniteGame.FortMinigameLogicComponent|GetCurrentMinigame (Variables|Default|GetFortMinigameProgress))))

(fn GetInventoryOverrides
  (bind _output_get (LocalVariables|SetSpawnParams (ItemData|MakeSpawnParamsList (Variables|Default|GetPickupItemList) (Math|Transform|MakeTransform "0, 0, 0"))))
  (Utilities|FlowControl|Sequence
    (:then_0
      (for _array_element _output_get
        (if (Utilities|IsValidSoftObjectReference (Utilities|Struct|BreakMinigameSpawnerSpawnParams _array_element))
          (Utilities|IsValid (Utilities|ResolveSoftReference (Utilities|Struct|BreakMinigameSpawnerSpawnParams _array_element))
            (:Is%20Valid
              (Utilities|Array|Add (LocalVariables|GetOutInventoryItems) (Utilities|Struct|MakeItemandCount (.PickupQuantity (Utilities|Struct|BreakMinigameSpawnerSpawnParams _array_element)) (Utilities|ResolveSoftReference (Utilities|Struct|BreakMinigameSpawnerSpawnParams _array_element)) (.ItemVariantGuid (Utilities|Struct|BreakMinigameSpawnerSpawnParams _array_element)))))))))
    (:then_1
      (return (LocalVariables|GetOutInventoryItems) (> (Utilities|Array|Length (LocalVariables|GetOutInventoryItems)) 0)))))

(fn GetVisibilityComponents
  (return (Utilities|Array|MakeArray (Variables|Default|GetBaseMesh) (Variables|Default|GetAmbient%20Loop))))

(fn GetVisibilityComponentsWithCollision
  (return (Utilities|Map|MakeMap (Variables|Default|GetBaseMesh) "QueryAndPhysics" (Variables|Default|GetPreviewStaticMesh) "NoCollision" (Variables|Default|GetPreviewSkeletalMesh) "NoCollision" (Variables|Default|GetAmbient%20Loop) "NoCollision")))

(fn GetWeaponSettingsMutator
  (params (Minigame :type Fort%20Minigame%20Object%20Reference))
  (bind _minigame (Variables|GetMinigame))
  (Utilities|FlowControl|Sequence
    (:then_0
      (Utilities|IsValid _minigame
        (:Is%20Valid
          (bind _output_get (LocalVariables|SetCreativeIslandMutatorList (Actor|GetComponentbyClass (Variables|/Script/FortniteGame.FortMinigame|GetVolume _minigame) "/Game/Athena/Playlists/Creative/Mutators/CreativeIslandMutatorList.CreativeIslandMutatorList_C")))
          (Utilities|IsValid _output_get
            (:Is%20Valid
              (bind _asfort_athena_mutator_weapon_settings (Utilities|Casting|CastToFortAthenaMutator_WeaponSettings (Modifier|GetMutatorbyClass _output_get "/Script/FortniteGame.FortAthenaMutator_WeaponSettings")))
              (LocalVariables|SetOutMutator _asfort_athena_mutator_weapon_settings))))))
    (:then_1
      (return (LocalVariables|GetOutMutator)))))

(fn HandleAddOrRemoveClassSettings
  (params (Minigame :type Fort%20Minigame%20Object%20Reference))
  (bind _classindex (Variables|DeviceOptions|GetClassIndex))
  (if (>= _classindex 0)
    (ClassSettings|AddClassSettingstoMinigame Minigame (Math|Conversions|ToByte(Integer) _classindex))
    (else
      (ClassSettings|RemoveClassSettingsfromMinigame Minigame))))

(fn HandleDropAreaOverlap
  (params (OverlappingActor :type Actor%20Object%20Reference))
  (if (not (Minigame|HasMinigameStarted))
    (bind _asfort_pickup_athena (Utilities|Casting|CastToFortPickupAthena OverlappingActor))
    (Utilities|IsPickupValid _asfort_pickup_athena (Variables|Default|GetInvalidPickupGameplayTagContainer)
      (:IsValid
        (if (not (Class|PickupItemListComponent|IsIteminQueue (Variables|Default|GetPickupItemList) _asfort_pickup_athena))
          (Inventory|AddItemtoQueue _asfort_pickup_athena))))))

(fn HasBeenGrantedItems
  (params (UniqueID :type Unique%20Net%20Id%20Structure))
  (return (Utilities|Set|ContainsItem (Variables|Default|GetGrantedPlayers) UniqueID)))

(fn HasMinigameStarted
  (bind _minigame (Minigame|GetCurrentMinigame))
  (Utilities|FlowControl|Sequence
    (:then_0
      (Utilities|IsValid _minigame
        (:Is%20Valid
          (LocalVariables|SetbSuccess (Minigame|HasStarted _minigame :!class /Script/FortniteGame.FortMinigame)))))
    (:then_1
      (return (LocalVariables|GetbSuccess)))))

(fn IsDeviceValidforPlayerState
  (params (PlayerState :type Fort%20Player%20State%20Object%20Reference))
  (return (Math|Boolean|ANDBoolean (CallFunction|IsPlayeronSameIsland PlayerState) (Minigame|IsMinigameinProgress) (>= (Variables|DeviceOptions|GetClassIndex) 0))))

(fn IsMinigameInProgress
  (bind _minigame (Minigame|GetCurrentMinigame))
  (Utilities|FlowControl|Sequence
    (:then_0
      (Utilities|IsValid _minigame
        (:Is%20Valid
          (LocalVariables|SetbSuccess (Minigame|IsinProgress _minigame)))))
    (:then_1
      (return (LocalVariables|GetbSuccess)))))

(fn IsPlayerMatchingClass
  (params (PlayerState :type Fort%20Player%20State%20Object%20Reference))
  (bind _minigame (Minigame|GetCurrentMinigame))
  (bind _returnvalue (Minigame|GetFNEClassComponent _minigame))
  (Utilities|FlowControl|Sequence
    (:then_0
      (bind _output_get (LocalVariables|SetMinigame _minigame))
      (Utilities|IsValid _output_get
        (:Is%20Valid
          (Utilities|IsValid _returnvalue
            (:Is%20Valid
              (LocalVariables|SetbSuccess (Minigame|IsPlayerUsingClassSlot _returnvalue (Variables|GetPlayerState) (Math|Conversions|ToByte(Integer) (Variables|DeviceOptions|GetClassIndex)))))))))
    (:then_1
      (return (LocalVariables|GetbSuccess)))))

(fn IsPlayerOnSameIsland
  (params (PlayerState :type Fort%20Player%20State%20Object%20Reference))
  (bind _returnvalue (Class|CreativeBlueprintLibrary|IsActorRelevanttoActor self PlayerState))
  (return _returnvalue))

(fn IsTargetableByLinearMover
  (return true))

(fn IsThisClass
  (params (Player%20Class :type Byte))
  (return (and (>= (Variables|DeviceOptions|GetClassIndex) 0) (Math|Byte|Equal(Byte) (Math|Conversions|ToByte(Integer) (Variables|DeviceOptions|GetClassIndex)) Player%20Class))))

(fn IsVisibleDuringMinigame
  (return (and (|CallParentFunction) (Variables|DeviceOptions|GetIsMeshVisibleDuringMinigame))))

(fn MaybeGiveAmmo
  (params (Inventory%20Owner :type Fort%20Inventory%20Owner%20Interface%20Interface) (ItemDefinition :type Fort%20World%20Item%20Definition%20Object%20Reference) (PickupInstigatorHandle :type Integer :default -1))
  (if (Variables|DeviceOptions|GetGrantExtraAmmoSetting)
    (HelperFunctions|GiveAmmotoInventoryOwner Inventory%20Owner ItemDefinition PickupInstigatorHandle (Variables|Default|GetAmmoCountMap) (Variables|Default|GetDefaultAmmoToGrant%20Const))))

(fn RemoveClassSettingsFromMinigame
  (params (Minigame :type Fort%20Minigame%20Object%20Reference))
  (Utilities|FlowControl|Sequence
    (:then_0
      (bind _output_get (LocalVariables|SetSettingsMutator (ClassSettings|GetClassSettingsMutator (Variables|GetMinigame))))
      (Utilities|IsValid _output_get
        (:Is%20Valid
          (bind _returnvalue (ClassSettings|RemoveDataSource _output_get (Variables|Default|GetToyOptionsComponent)))
          (bind _returnvalue_2 (ClassSettings|RemoveDataSource (LocalVariables|GetSettingsMutator) self))
          (if (or _returnvalue_2 _returnvalue)
            (Minigame|GameRules|RefreshGameRulesPreview (Variables|GetMinigame))))))
    (:then_1
      (bind _output_get_2 (LocalVariables|SetWeaponSettingsMutator (ClassSettings|GetWeaponSettingsMutator (Variables|GetMinigame))))
      (Utilities|IsValid _output_get_2
        (:Is%20Valid
          (Class|FortAthenaMutatorWeaponSettings|RemoveDataSource _output_get_2 (Variables|Default|GetToyOptionsComponent)))))
    (:then_2
      (Class|FortAthenaMutatorUtilities|RemoveDataRegistryOptionsforContextObject))))

(fn RemoveItemsFromPlayer
  (params (PlayerController :type Fort%20Player%20Controller%20Object%20Reference))
  (Utilities|IsValid PlayerController
    (:Is%20Valid
      (for _array_element (ItemData|MakeSpawnParamsList (Variables|Default|GetPickupItemList) (LocalVariables|GetSpawnTransform))
        (Inventory|RemoveItemsFromPlayerByIntStateValue (Variables|GetPlayerController) "PickupInstigatorHandle" (.PickupInstigatorHandle (Utilities|Struct|BreakMinigameSpawnerSpawnParams _array_element)) true)))))

(fn ShouldGrantItems
  (params (UniqueID :type Unique%20Net%20Id%20Structure))
  (return (or (not (Inventory|HasBeenGrantedItems UniqueID)) (Variables|DeviceOptions|GetGrantOnRespawn))))

(fn ShouldInvisiblePropsBeAllowed
  (return false))

(fn TryGiveAllItemsToPlayer
  (params (PlayerController :type Fort%20Player%20Controller%20Athena%20Object%20Reference))
  (bind _pickupitemlist (Variables|Default|GetPickupItemList))
  (bind _returnvalue (ItemData|MakeSpawnParamsList _pickupitemlist (struct :Location "0, 0, 0" :Rotation "0, 0, 0" :Scale "1.000000,1.000000,1.000000")))
  (bind _inventoryowner (LocalVariables|GetInventoryOwner))
  (Inventory|RemoveItemsfromPlayer PlayerController)
  (Utilities|FlowControl|Sequence
    (:then_0
      (Variables|/Script/FortniteGame.FortPlayerControllerAthena|SetbShouldForceDeleteDroppedItems true (Variables|GetPlayerController)))
    (:then_1
      (bind _asfort_inventory_owner_interface (Utilities|Casting|CastToFortInventoryOwnerInterface (Variables|GetPlayerController)))
      (LocalVariables|SetInventoryOwner _asfort_inventory_owner_interface)
      (for (_array_element _array_index) _returnvalue
        (Utilities|IsValid (Utilities|ResolveSoftReference (Utilities|Struct|BreakMinigameSpawnerSpawnParams _array_element))
          (:Is%20Valid
            (Inventory|GiveItemtoInventoryOwner _inventoryowner (Utilities|ResolveSoftReference (Utilities|Struct|BreakMinigameSpawnerSpawnParams _array_element)) (.ItemVariantGuid (Utilities|Struct|BreakMinigameSpawnerSpawnParams _array_element)) (.PickupQuantity (Utilities|Struct|BreakMinigameSpawnerSpawnParams _array_element)) true -1 (.PickupInstigatorHandle (Utilities|Struct|BreakMinigameSpawnerSpawnParams _array_element)))
            (Inventory|MaybeGiveAmmo _inventoryowner (Utilities|ResolveSoftReference (Utilities|Struct|BreakMinigameSpawnerSpawnParams _array_element)) (.PickupInstigatorHandle (Utilities|Struct|BreakMinigameSpawnerSpawnParams _array_element)))
            (if (== (Math|Integer|Min(Integer) (Utilities|Array|LastIndex _returnvalue) (Variables|Default|GetEquipItemOnGrant)) _array_index)
              (Inventory|ExecuteInventoryItemDefinition :self (Variables|GetPlayerController) :ItemDefinition (Utilities|ResolveSoftReference (Utilities|Struct|BreakMinigameSpawnerSpawnParams _array_element)) :Delay 0.0 :bForceExecute true :bActivateSlotAfterSettingFocused true))))))
    (:then_2
      (Variables|/Script/FortniteGame.FortPlayerControllerAthena|SetbShouldForceDeleteDroppedItems false (Variables|GetPlayerController)))))

(fn ConstructionScript
  (Networking|CanExecuteCosmeticEvents
    (:True
      (Variables|BuildingActor|SetOverridePrimitivesToExcludeFoliage (Utilities|Array|MakeArray (Variables|Default|GetBaseMesh))))))
